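# InviteEventTitle is a UserWidget that shows a title image in the account select area of the Fortnite login UI.
# A root VerticalBox holds a single CommonLazyImage, TitleImage, centered with negative horizontal padding and set to the Fortnite wordmark logo texture.
# The event graph is empty, so the widget contains no logic of its own.
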WidgetBlueprint /FortniteCoreUI/UI/Login/AccountSelect/InviteEventTitle
graphs [EventGraph {links [], nodes []}]
variables []
widgets {
 info {inheritedWidgetCount=0, parentClass=&UserWidget, rootWidgetClass=&VerticalBox, widgetCount=2}
 widgets [
  VerticalBox_0 {widgetClassPath=&VerticalBox, widget=VerticalBox_0}
  TitleImage {
   widgetClassPath=/Script/CommonUI.CommonLazyImage
   bIsVariable=true
   parent=VerticalBox_0
   widget=TitleImage
   properties {
    visibility=SelfHitTestInvisible
    brush {
     bIsDynamicallyLoaded=false
     drawAs=Image
     imageType=NoImage
     mirroring=NoMirror
     resourceName=None
     resourceObject=/FortniteCoreUI/Foundation/Textures/Logo/T-Logo-Wordmark-1024x256
     tiling=NoTile
     imageSize=<600 158>
     margin=<ltrb 0 0 0 0>
     outlineSettings {
      bUseBrushTransparency=false
      roundingType=HalfHeightRadius
      width=0
      color {colorUseRule=UseColor_Specified, specifiedColor=<rgba 0 0 0 0>}
      cornerRadii=<0 0 0 0>
     }
     tintColor {colorUseRule=UseColor_Specified, specifiedColor=<rgba 1 1 1 1>}
     uVRegion {bIsValid=false, max=<0 0>, min=<0 0>}
    }
   }
   slot {class=VerticalBoxSlot, properties {horizontalAlignment=HAlign_Center, verticalAlignment=VAlign_Center, padding=<ltrb -30 0 -30 0>}}
  }
 ]
}
