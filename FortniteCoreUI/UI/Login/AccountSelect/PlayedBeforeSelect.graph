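# Login screen widget that asks whether the player has played Fortnite before, offering Yes (find account) or No (create account).
# A switcher holds the choice layout and an embedded SignInWidget, beneath a collapsed blur and spinner overlay for pending states.
# OnActivated_1 plays the staggered Intro animation on title and buttons.
# Button clicks only fire an element-interacted event; the graph leaves selection and switching to other code, based on native FortPlayedBeforeSelect.
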WidgetBlueprint /FortniteCoreUI/UI/Login/AccountSelect/PlayedBeforeSelect
animations [Intro {
  binding_count=7
  display_label=""
  length_seconds=2.3
  object=&self:Intro
  sequence {bindings [
    CommonTextBlock_1 {
     id=#1
     possessed_class=CommonTextBlock
     tracks [
      {
       type=MovieSceneColorTrack
       display_name=Color and Opacity
       sections [{
         type=MovieSceneColorSection
         channels [
          Color.R {type=MovieSceneScriptingFloatChannel}
          Color.G {type=MovieSceneScriptingFloatChannel}
          Color.B {type=MovieSceneScriptingFloatChannel}
          Color.A {type=MovieSceneScriptingFloatChannel, keys [[29,0],[33,1]]}
         ]
        }]
      }
      {
       type=MovieScene2DTransformTrack
       sections [{
         type=MovieScene2DTransformSection
         channels [
          Translation.X {type=MovieSceneScriptingFloatChannel, keys [[30.9997,500],[34.9997,-2.5429],[35.9997,0]]}
          Translation.Y {type=MovieSceneScriptingFloatChannel, const [0,1]}
          Angle {type=MovieSceneScriptingFloatChannel}
          Scale.X {type=MovieSceneScriptingFloatChannel}
          Scale.Y {type=MovieSceneScriptingFloatChannel}
          Shear.X {type=MovieSceneScriptingFloatChannel}
          Shear.Y {type=MovieSceneScriptingFloatChannel}
         ]
        }]
      }
     ]
    }
    HorizontalBoxChoices {
     id=#2
     possessed_class=HorizontalBox
     tracks [
      {
       type=MovieScene2DTransformTrack
       sections [{
         type=MovieScene2DTransformSection
         end_frame=46
         start_frame=0
         channels [
          Translation.X {type=MovieSceneScriptingFloatChannel, const [0,3]}
          Translation.Y {type=MovieSceneScriptingFloatChannel, keys [[7.9997,5.5007],[25,5.5007],[30,0]]}
          Angle {type=MovieSceneScriptingFloatChannel}
          Scale.X {type=MovieSceneScriptingFloatChannel, keys [[24.9997,0.4],[27.9997,1],[30.9997,1]]}
          Scale.Y {type=MovieSceneScriptingFloatChannel, keys [[24.9997,0.4],[27.9997,1],[30.9997,1]]}
          Shear.X {type=MovieSceneScriptingFloatChannel}
          Shear.Y {type=MovieSceneScriptingFloatChannel}
         ]
        }]
      }
      {
       type=MovieSceneFloatTrack
       display_name=Render Opacity
       sections [{type=MovieSceneFloatSection, channels [None {type=MovieSceneScriptingFloatChannel, keys [[25,0],[35.9997,1]]}]}]
      }
     ]
    }
    CommonTextBlock_3 {
     id=#3
     possessed_class=CommonTextBlock
     tracks [{
       type=MovieScene2DTransformTrack
       sections [{
         type=MovieScene2DTransformSection
         end_frame=46
         start_frame=0
         channels [
          Translation.X {type=MovieSceneScriptingFloatChannel, keys [[33,123],[35.9997,0],[36.9997,0]]}
          Translation.Y {type=MovieSceneScriptingFloatChannel}
          Angle {type=MovieSceneScriptingFloatChannel}
          Scale.X {type=MovieSceneScriptingFloatChannel}
          Scale.Y {type=MovieSceneScriptingFloatChannel}
          Shear.X {type=MovieSceneScriptingFloatChannel}
          Shear.Y {type=MovieSceneScriptingFloatChannel}
         ]
        }]
      }]
    }
    CommonTextBlock_6 {
     id=#4
     possessed_class=CommonTextBlock
     tracks [
      {
       type=MovieSceneColorTrack
       display_name=Color and Opacity
       sections [{
         type=MovieSceneColorSection
         channels [
          Color.R {type=MovieSceneScriptingFloatChannel}
          Color.G {type=MovieSceneScriptingFloatChannel}
          Color.B {type=MovieSceneScriptingFloatChannel}
          Color.A {type=MovieSceneScriptingFloatChannel, keys [[28.9997,0],[32.9997,1]]}
         ]
        }]
      }
      {
       type=MovieScene2DTransformTrack
       sections [{
         type=MovieScene2DTransformSection
         channels [
          Translation.X {type=MovieSceneScriptingFloatChannel, keys [[31.9997,500],[35.9997,-2.5429],[36.9997,0]]}
          Translation.Y {type=MovieSceneScriptingFloatChannel, const [0,1]}
          Angle {type=MovieSceneScriptingFloatChannel}
          Scale.X {type=MovieSceneScriptingFloatChannel}
          Scale.Y {type=MovieSceneScriptingFloatChannel}
          Shear.X {type=MovieSceneScriptingFloatChannel}
          Shear.Y {type=MovieSceneScriptingFloatChannel}
         ]
        }]
      }
     ]
    }
    TitleImage {
     id=#5
     display_name=InviteEventTitle
     possessed_class=CommonLazyImage
     tracks [
      {
       type=MovieSceneFloatTrack
       display_name=Render Opacity
       sections [{type=MovieSceneFloatSection, channels [None {type=MovieSceneScriptingFloatChannel, keys [[6.9997,0],[15,1]]}]}]
      }
      {
       type=MovieScene2DTransformTrack
       sections [{
         type=MovieScene2DTransformSection
         channels [
          Translation.X {type=MovieSceneScriptingFloatChannel}
          Translation.Y {type=MovieSceneScriptingFloatChannel, thinned_from=4, keys [[6.9997,518.04],[10.9997,0],[12.9997,0]]}
          Angle {type=MovieSceneScriptingFloatChannel}
          Scale.X {type=MovieSceneScriptingFloatChannel, keys [[6.9997,0.5],[11,1.2],[13.9997,1]]}
          Scale.Y {type=MovieSceneScriptingFloatChannel, keys [[6.9997,0.5],[11,1.2],[13.9997,1]]}
          Shear.X {type=MovieSceneScriptingFloatChannel}
          Shear.Y {type=MovieSceneScriptingFloatChannel}
         ]
        }]
      }
     ]
    }
    CommonTextBlock_5 {
     id=#6
     possessed_class=CommonTextBlock
     tracks [{
       type=MovieScene2DTransformTrack
       sections [{
         type=MovieScene2DTransformSection
         end_frame=46
         start_frame=0
         channels [
          Translation.X {type=MovieSceneScriptingFloatChannel, keys [[34,123],[36.9997,0],[37.9997,0]]}
          Translation.Y {type=MovieSceneScriptingFloatChannel}
          Angle {type=MovieSceneScriptingFloatChannel}
          Scale.X {type=MovieSceneScriptingFloatChannel}
          Scale.Y {type=MovieSceneScriptingFloatChannel}
          Shear.X {type=MovieSceneScriptingFloatChannel}
          Shear.Y {type=MovieSceneScriptingFloatChannel}
         ]
        }]
      }]
    }
    BorderTitleText {
     id=#7
     possessed_class=Border
     tracks [
      {
       type=MovieScene2DTransformTrack
       sections [{
         type=MovieScene2DTransformSection
         channels [
          Translation.X {type=MovieSceneScriptingFloatChannel}
          Translation.Y {type=MovieSceneScriptingFloatChannel, keys [[7.9997,110.152],[24.3473,110.2721],[24.6717,110.8705],[27.9997,-8.0542],[29,0]]}
          Angle {type=MovieSceneScriptingFloatChannel, keys [[7.9997,-3],[10.9997,0]]}
          Scale.X {type=MovieSceneScriptingFloatChannel, thinned_from=4, keys [[7.9997,5.7993],[10.9997,1],[27.9997,1]]}
          Scale.Y {type=MovieSceneScriptingFloatChannel, thinned_from=4, keys [[7.9997,5.7993],[10.9997,1],[27.9997,1]]}
          Shear.X {type=MovieSceneScriptingFloatChannel}
          Shear.Y {type=MovieSceneScriptingFloatChannel}
         ]
        }]
      }
      {
       type=MovieSceneColorTrack
       display_name=Content Color and Opacity
       sections [{
         type=MovieSceneColorSection
         channels [
          Color.R {type=MovieSceneScriptingFloatChannel}
          Color.G {type=MovieSceneScriptingFloatChannel}
          Color.B {type=MovieSceneScriptingFloatChannel}
          Color.A {type=MovieSceneScriptingFloatChannel, keys [[7.9997,0],[13.9997,1],[27.9997,1]]}
         ]
        }]
      }
     ]
    }
   ]}
 }]
graphs [EventGraph {nodes [
   n1 = "On Button Clicked (Button_Yes)"!() -> (OutputDelegate, then, Button:Common Button Legacy Object Reference) {is=title K2Node_ComponentBoundEvent}
   > Fire Event Element Interacted(Element=@n1.Button)
   n2 = "On Button Clicked (Button_No)"!() -> (OutputDelegate, then, Button:Common Button Legacy Object Reference) {is=title K2Node_ComponentBoundEvent}
   > Fire Event Element Interacted(Element=@n2.Button)
   K2Node_Event_0 = OnActivated_1() {is=title K2Node_CustomEvent}
   > PlayAnimation(InAnimation=Get Intro())
  ]}]
variables []
widgets {
 info {
  inheritedWidgetCount=3
  namedSlotCount=2
  parentClass=/Script/FortniteUI.FortPlayedBeforeSelect
  rootWidgetClass=&SafeZone
  widgetCount=33
 }
 widgets [
  SafeZone_0 {widgetClassPath=&SafeZone, bIsVariable=true, widget=SafeZone_0, properties {padLeft=false, padRight=false, padTop=false}}
  Overlay_0 {widgetClassPath=&Overlay, parent=SafeZone_0, widget=Overlay_0, slot {class=SafeZoneSlot}}
  Border_44 {
   widgetClassPath=&Border
   parent=Overlay_0
   widget=Border_44
   properties {background {
     bIsDynamicallyLoaded=false
     drawAs=Image
     imageType=NoImage
     mirroring=NoMirror
     resourceName=None
     resourceObject=null
     tiling=NoTile
     imageSize=<32 32>
     margin=<ltrb 0 0 0 0>
     outlineSettings {
      bUseBrushTransparency=false
      roundingType=HalfHeightRadius
      width=0
      color {colorUseRule=UseColor_Specified, specifiedColor=<rgba 0 0 0 0>}
      cornerRadii=<0 0 0 0>
     }
     tintColor {colorUseRule=UseColor_Specified, specifiedColor=<rgba 0 0 0 0.6>}
     uVRegion {bIsValid=false, max=<0 0>, min=<0 0>}
    }}
   slot {class=OverlaySlot, properties {horizontalAlignment=HAlign_Fill, verticalAlignment=VAlign_Fill, padding=<ltrb -200 -200 -200 -200>}}
  }
  SizeBox_0 {
   widgetClassPath=&SizeBox
   parent=Overlay_0
   widget=SizeBox_0
   properties {bOverride_HeightOverride=true, bOverride_WidthOverride=true, heightOverride=760, widthOverride=1125}
   slot {class=OverlaySlot, properties {horizontalAlignment=HAlign_Center, verticalAlignment=VAlign_Bottom, padding=<ltrb 0 0 0 50>}}
  }
  LoginType_Switcher {widgetClassPath=&WidgetSwitcher, bIsVariable=true, parent=SizeBox_0, widget=LoginType_Switcher, slot {class=SizeBoxSlot}}
  VerticalBox_0 {
   widgetClassPath=&VerticalBox
   parent=LoginType_Switcher
   widget=VerticalBox_0
   slot {class=WidgetSwitcherSlot, properties {verticalAlignment=VAlign_Bottom}}
  }
  VerticalBox {
   widgetClassPath=&VerticalBox
   parent=VerticalBox_0
   widget=VerticalBox
   slot {class=VerticalBoxSlot, properties {size {sizeRule=Fill, value=1}}}
  }
  HorizontalBox_1 {
   widgetClassPath=&HorizontalBox
   parent=VerticalBox
   widget=HorizontalBox_1
   slot {class=VerticalBoxSlot, properties {horizontalAlignment=HAlign_Center, verticalAlignment=VAlign_Top, padding=<ltrb 0 0 0 50>}}
  }
  InviteEventTitle {
   widgetClassPath=&/InviteEventTitle
   bIsVariable=true
   parent=HorizontalBox_1
   widget=InviteEventTitle
   properties {renderOpacity=0, visibility=Hidden, renderTransformPivot=<0.5 1>}
   slot {class=HorizontalBoxSlot}
  }
  VerticalBox_1 {
   widgetClassPath=&VerticalBox
   parent=VerticalBox
   widget=VerticalBox_1
   slot {class=VerticalBoxSlot, properties {horizontalAlignment=HAlign_Center, padding=<ltrb 0 50 0 0>}}
  }
  BorderTitleText {
   widgetClassPath=&Border
   parent=VerticalBox_1
   widget=BorderTitleText
   properties {
    horizontalAlignment=HAlign_Left
    visibility=HitTestInvisible
    background {
     bIsDynamicallyLoaded=false
     drawAs=NoDrawType
     imageType=NoImage
     mirroring=NoMirror
     resourceName=None
     resourceObject=null
     tiling=NoTile
     imageSize=<32 32>
     margin=<ltrb 0 0 0 0>
     outlineSettings {
      bUseBrushTransparency=false
      roundingType=HalfHeightRadius
      width=0
      color {colorUseRule=UseColor_Specified, specifiedColor=<rgba 0 0 0 0>}
      cornerRadii=<0 0 0 0>
     }
     tintColor {colorUseRule=UseColor_Specified, specifiedColor=<rgba 1 1 1 1>}
     uVRegion {bIsValid=false, max=<0 0>, min=<0 0>}
    }
    padding=<ltrb 0 0 0 0>
   }
   slot {class=VerticalBoxSlot}
  }
  CommonWidgetSwitcher_InviteOnly {
   widgetClassPath=&CommonWidgetSwitcherLegacy
   bIsVariable=true
   parent=BorderTitleText
   widget=CommonWidgetSwitcher_InviteOnly
   slot {class=BorderSlot, properties {horizontalAlignment=HAlign_Left, padding=<ltrb 0 0 0 0>}}
  }
  CommonTextTitle {
   widgetClassPath=&CommonTextBlock
   parent=CommonWidgetSwitcher_InviteOnly
   widget=CommonTextTitle
   properties {
    text="HAVE YOU PLAYED FORTNITE BEFORE?\r\nIF SO, THEN WE CAN FIND YOUR ACCOUNT WITH ALL YOUR PROGRESS AND PURCHASES."
    wrapTextAt=1050
    font {
     bForceMonospaced=false
     bMaterialIsStencil=false
     fontMaterial=null
     fontObject=&NotoSans
     letterSpacing=0
     monospacedWidth=1
     size=24
     skewAmount=0
     typefaceFontName=NotoSans-Regular
     outlineSettings {
      bApplyOutlineToDropShadows=false
      bMiteredCorners=false
      bSeparateFillAlpha=false
      outlineMaterial=null
      outlineSize=0
      outlineColor=<rgba 0 0 0 1>
     }
    }
    shadowOffset=<0 0>
   }
   slot {class=WidgetSwitcherSlot, properties {padding=<ltrb 30 0 0 0>}}
  }
  CommonTextBlock_4 {
   widgetClassPath=&CommonTextBlock
   parent=CommonWidgetSwitcher_InviteOnly
   widget=CommonTextBlock_4
   properties {
    style=/FortUILibrary/Styles/Typography/Foundation/TextStyle-Header-L.TextStyle-Header-L_C
    text=HAVE YOU PLAYED FORTNITE BEFORE?
    wrapTextAt=850
    font {
     bForceMonospaced=false
     bMaterialIsStencil=false
     fontMaterial=null
     fontObject=&Burbank
     letterSpacing=0
     monospacedWidth=1
     size=48
     skewAmount=0
     typefaceFontName=Burbank-Black
     outlineSettings {
      bApplyOutlineToDropShadows=false
      bMiteredCorners=false
      bSeparateFillAlpha=false
      outlineMaterial=null
      outlineSize=0
      outlineColor=<rgba 0 0 0 1>
     }
    }
    shadowOffset=<0 0>
   }
   slot {class=WidgetSwitcherSlot}
  }
  HorizontalBoxChoices {
   widgetClassPath=&HorizontalBox
   parent=VerticalBox_1
   widget=HorizontalBoxChoices
   slot {class=VerticalBoxSlot, properties {verticalAlignment=VAlign_Top, size {sizeRule=Fill, value=1}}}
  }
  Button_Yes {
   widgetClassPath=/FortUILibrary/Foundation/Buttons/SkewButton
   bInherited=true
   bIsVariable=true
   parent=HorizontalBoxChoices
   widget=Button_Yes
   properties {
    bIsFocusable=true
    clipping=ClipToBounds
    onButtonBaseClicked=(&self:WidgetTree.Button_Yes.HandleOnButtonClicked)
    onButtonBaseDoubleClicked=(&self:WidgetTree.Button_Yes.HandleOnButtonDoubleClicked)
    onButtonBaseHovered=(&self:WidgetTree.Button_Yes.HandleOnButtonHovered)
    onButtonBaseUnhovered=(&self:WidgetTree.Button_Yes.HandleOnButtonUnhovered)
    onSelectedChangedBase=(&self:WidgetTree.Button_Yes.HandleOnSelectedChanged)
    buttonColor=<rgba 0.1779 1 0.0704 1>
   }
   slot {class=HorizontalBoxSlot}
  }
  CommonBorder_1 {
   widgetClassPath=&CommonBorder
   namedSlotHost=Button_Yes
   widget=CommonBorder_1
   properties {contentColorAndOpacity=<rgba 0.0144 0.0144 0.0156 1>, padding=<ltrb 0 30 0 30>}
  }
  Overlay_2 {widgetClassPath=&Overlay, parent=CommonBorder_1, widget=Overlay_2, slot {class=BorderSlot, properties {padding=<ltrb 0 30 0 30>}}}
  HorizontalBox_5 {widgetClassPath=&HorizontalBox, parent=Overlay_2, widget=HorizontalBox_5, slot {class=OverlaySlot}}
  CommonTextBlock_3 {
   widgetClassPath=&CommonTextBlock
   parent=HorizontalBox_5
   widget=CommonTextBlock_3
   properties {
    text=YES
    font {
     bForceMonospaced=false
     bMaterialIsStencil=false
     fontMaterial=null
     fontObject=&Burbank
     letterSpacing=0
     monospacedWidth=1
     size=91
     skewAmount=0
     typefaceFontName=Burbank-Black
     outlineSettings {
      bApplyOutlineToDropShadows=false
      bMiteredCorners=false
      bSeparateFillAlpha=false
      outlineMaterial=null
      outlineSize=0
      outlineColor=<rgba 0 0 0 1>
     }
    }
   }
   slot {class=HorizontalBoxSlot}
  }
  CommonTextBlock_1 {
   widgetClassPath=&CommonTextBlock
   bIsVariable=true
   parent=HorizontalBox_5
   widget=CommonTextBlock_1
   properties {
    scrollStyle=&TextScrollStyle-Base_C
    style=/FortUILibrary/Styles/Typography/Foundation/TextStyle-Base-M.TextStyle-Base-M_C
    text=Find my account
    wrapTextAt=250.3858
    font {
     bForceMonospaced=false
     bMaterialIsStencil=false
     fontMaterial=null
     fontObject=&NotoSans
     letterSpacing=0
     monospacedWidth=1
     size=24
     skewAmount=0
     typefaceFontName=NotoSans-Regular
     outlineSettings {
      bApplyOutlineToDropShadows=false
      bMiteredCorners=false
      bSeparateFillAlpha=false
      outlineMaterial=null
      outlineSize=0
      outlineColor=<rgba 0 0 0 1>
     }
    }
    shadowOffset=<0 0>
   }
   slot {class=HorizontalBoxSlot, properties {horizontalAlignment=HAlign_Left, verticalAlignment=VAlign_Center, padding=<ltrb 30 0 0 0>}}
  }
  ImageExternalLink_1 {
   widgetClassPath=&Image
   parent=Overlay_2
   widget=ImageExternalLink_1
   properties {brush {
     bIsDynamicallyLoaded=false
     drawAs=Image
     imageType=NoImage
     mirroring=NoMirror
     resourceName=None
     resourceObject=/FortniteCoreUI/UI/Login/Art/M-Icon-ExternalLink-64
     tiling=NoTile
     imageSize=<75 75>
     margin=<ltrb 0 0 0 0>
     outlineSettings {
      bUseBrushTransparency=false
      roundingType=HalfHeightRadius
      width=0
      color {colorUseRule=UseColor_Specified, specifiedColor=<rgba 0 0 0 0>}
      cornerRadii=<0 0 0 0>
     }
     tintColor {colorUseRule=UseColor_Specified, specifiedColor=<rgba 1 1 1 1>}
     uVRegion {bIsValid=false, max=<0 0>, min=<0 0>}
    }}
   slot {class=OverlaySlot, properties {horizontalAlignment=HAlign_Right, verticalAlignment=VAlign_Bottom, padding=<ltrb 0 0 -40 -10>}}
  }
  Button_No {
   widgetClassPath=&SkewButton
   bInherited=true
   bIsVariable=true
   parent=HorizontalBoxChoices
   widget=Button_No
   properties {
    bIsFocusable=true
    clipping=ClipToBounds
    onButtonBaseClicked=(&self:WidgetTree.Button_No.HandleOnButtonClicked)
    onButtonBaseDoubleClicked=(&self:WidgetTree.Button_No.HandleOnButtonDoubleClicked)
    onButtonBaseHovered=(&self:WidgetTree.Button_No.HandleOnButtonHovered)
    onButtonBaseUnhovered=(&self:WidgetTree.Button_No.HandleOnButtonUnhovered)
    onSelectedChangedBase=(&self:WidgetTree.Button_No.HandleOnSelectedChanged)
    buttonColor=<rgba 0.0076 0.3023 0.651 1>
    foregroundColor {colorUseRule=UseColor_Foreground, specifiedColor=<rgba 1 1 1 1>}
   }
   slot {class=HorizontalBoxSlot}
  }
  CommonBorder_2 {widgetClassPath=&CommonBorder, namedSlotHost=Button_No, widget=CommonBorder_2, properties {padding=<ltrb 0 30 0 30>}}
  Overlay_1 {widgetClassPath=&Overlay, parent=CommonBorder_2, widget=Overlay_1, slot {class=BorderSlot, properties {padding=<ltrb 0 30 0 30>}}}
  HorizontalBox_6 {widgetClassPath=&HorizontalBox, parent=Overlay_1, widget=HorizontalBox_6, slot {class=OverlaySlot}}
  CommonTextBlock_5 {
   widgetClassPath=&CommonTextBlock
   parent=HorizontalBox_6
   widget=CommonTextBlock_5
   properties {
    text=NO
    font {
     bForceMonospaced=false
     bMaterialIsStencil=false
     fontMaterial=null
     fontObject=&Burbank
     letterSpacing=0
     monospacedWidth=1
     size=91
     skewAmount=0
     typefaceFontName=Burbank-Black
     outlineSettings {
      bApplyOutlineToDropShadows=false
      bMiteredCorners=false
      bSeparateFillAlpha=false
      outlineMaterial=null
      outlineSize=0
      outlineColor=<rgba 0 0 0 1>
     }
    }
   }
   slot {class=HorizontalBoxSlot}
  }
  CommonTextBlock_6 {
   widgetClassPath=&CommonTextBlock
   bIsVariable=true
   parent=HorizontalBox_6
   widget=CommonTextBlock_6
   properties {
    scrollStyle=&TextScrollStyle-Base_C
    style=&TextStyle-Base-M_C
    text=Create a new account
    wrapTextAt=250.3858
    font {
     bForceMonospaced=false
     bMaterialIsStencil=false
     fontMaterial=null
     fontObject=&NotoSans
     letterSpacing=0
     monospacedWidth=1
     size=24
     skewAmount=0
     typefaceFontName=NotoSans-Regular
     outlineSettings {
      bApplyOutlineToDropShadows=false
      bMiteredCorners=false
      bSeparateFillAlpha=false
      outlineMaterial=null
      outlineSize=0
      outlineColor=<rgba 0 0 0 1>
     }
    }
    shadowOffset=<0 0>
   }
   slot {class=HorizontalBoxSlot, properties {horizontalAlignment=HAlign_Left, verticalAlignment=VAlign_Center, padding=<ltrb 30 0 0 0>}}
  }
  ImageExternalLink {
   widgetClassPath=&Image
   parent=Overlay_1
   widget=ImageExternalLink
   properties {brush {
     bIsDynamicallyLoaded=false
     drawAs=Image
     imageType=NoImage
     mirroring=NoMirror
     resourceName=None
     resourceObject=&M-Icon-ExternalLink-64
     tiling=NoTile
     imageSize=<75 75>
     margin=<ltrb 0 0 0 0>
     outlineSettings {
      bUseBrushTransparency=false
      roundingType=HalfHeightRadius
      width=0
      color {colorUseRule=UseColor_Specified, specifiedColor=<rgba 0 0 0 0>}
      cornerRadii=<0 0 0 0>
     }
     tintColor {colorUseRule=UseColor_Specified, specifiedColor=<rgba 1 1 1 1>}
     uVRegion {bIsValid=false, max=<0 0>, min=<0 0>}
    }}
   slot {class=OverlaySlot, properties {horizontalAlignment=HAlign_Right, verticalAlignment=VAlign_Bottom, padding=<ltrb 0 0 -40 -10>}}
  }
  HorizontalBox_0 {
   widgetClassPath=&HorizontalBox
   parent=VerticalBox_0
   widget=HorizontalBox_0
   slot {
    class=VerticalBoxSlot
    properties {horizontalAlignment=HAlign_Left, verticalAlignment=VAlign_Bottom, padding=<ltrb -45 0 0 0>, size {sizeRule=Fill, value=0.2}}
   }
  }
  SignInWidget {
   widgetClassPath=/FortniteCoreUI/UI/Login/SignIn/SignInWidget
   bIsVariable=true
   parent=LoginType_Switcher
   widget=SignInWidget
   slot {class=WidgetSwitcherSlot}
  }
  BackgroundBlur_PendingSpinner {
   widgetClassPath=&BackgroundBlur
   bInherited=true
   bIsVariable=true
   parent=Overlay_0
   widget=BackgroundBlur_PendingSpinner
   properties {
    blurStrength=3.5
    horizontalAlignment=HAlign_Center
    verticalAlignment=VAlign_Center
    visibility=Collapsed
    lowQualityFallbackBrush {
     bIsDynamicallyLoaded=false
     drawAs=Image
     imageType=NoImage
     mirroring=NoMirror
     resourceName=None
     resourceObject=null
     tiling=NoTile
     imageSize=<1024 768>
     margin=<ltrb 0 0 0 0>
     outlineSettings {
      bUseBrushTransparency=false
      roundingType=HalfHeightRadius
      width=0
      color {colorUseRule=UseColor_Specified, specifiedColor=<rgba 0 0 0 0>}
      cornerRadii=<0 0 0 0>
     }
     tintColor {colorUseRule=UseColor_Specified, specifiedColor=<rgba 0.0052 0.0052 0.0056 0.5>}
     uVRegion {bIsValid=false, max=<0 0>, min=<0 0>}
    }
    padding=<ltrb 0 0 0 80>
   }
   slot {class=OverlaySlot, properties {horizontalAlignment=HAlign_Fill, verticalAlignment=VAlign_Fill, padding=<ltrb -500 -500 -500 -500>}}
  }
  Spinner {
   widgetClassPath=&Image
   parent=BackgroundBlur_PendingSpinner
   widget=Spinner
   properties {
    brush {
     bIsDynamicallyLoaded=false
     drawAs=Image
     imageType=NoImage
     mirroring=NoMirror
     resourceName=None
     resourceObject=/FortUILibrary/UIKit/Throbber/MI_UIKit_Monochrome_Throbber
     tiling=NoTile
     imageSize=<32 32>
     margin=<ltrb 0 0 0 0>
     outlineSettings {
      bUseBrushTransparency=false
      roundingType=HalfHeightRadius
      width=0
      color {colorUseRule=UseColor_Specified, specifiedColor=<rgba 0 0 0 0>}
      cornerRadii=<0 0 0 0>
     }
     tintColor {colorUseRule=UseColor_Specified, specifiedColor=<rgba 1 1 1 1>}
     uVRegion {bIsValid=false, max=<0 0>, min=<0 0>}
    }
    renderTransform {angle=0, scale=<2.5 2.5>, shear=<0 0>, translation=<0 0>}
   }
   slot {class=BackgroundBlurSlot, properties {horizontalAlignment=HAlign_Center, verticalAlignment=VAlign_Center, padding=<ltrb 0 0 0 80>}}
  }
 ]
}
