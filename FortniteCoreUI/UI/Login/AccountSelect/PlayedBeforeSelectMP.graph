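# Start-login widget based on FortStartLoginWidget, showing a "Sign in to Epic Games" button and a collapsed blurred pending spinner.
# Activation plays an intro that slides and fades in the button.
# On construct and on PlatformGameInstance orientation changes (unbound on destruct), it re-lays out the button: padded full-width for tall portrait screens, centered otherwise.
# The button's click handler is empty in this graph.
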WidgetBlueprint /FortniteCoreUI/UI/Login/AccountSelect/PlayedBeforeSelectMP
animations [Intro {
  binding_count=1
  length_seconds=0.25
  object=&self:Intro
  sequence {bindings [Button_Login {
     id=#1
     possessed_class=WBP_UIKit_Button_Loud_C
     tracks [
      {
       type=MovieScene2DTransformTrack
       sections [{
         type=MovieScene2DTransformSection
         end_frame=5
         start_frame=0
         channels [
          Translation.X {type=MovieSceneScriptingFloatChannel, const [0,2]}
          Translation.Y {type=MovieSceneScriptingFloatChannel, keys [[0,128],[5,0]]}
          Angle {type=MovieSceneScriptingFloatChannel, const [0,1]}
          Scale.X {type=MovieSceneScriptingFloatChannel, const [1,1]}
          Scale.Y {type=MovieSceneScriptingFloatChannel, const [1,1]}
          Shear.X {type=MovieSceneScriptingFloatChannel, const [0,1]}
          Shear.Y {type=MovieSceneScriptingFloatChannel, const [0,1]}
         ]
        }]
      }
      {
       type=MovieSceneFloatTrack
       display_name=Render Opacity
       sections [{type=MovieSceneFloatSection, end_frame=5, start_frame=0, channels [None {type=MovieSceneScriptingFloatChannel, keys [[0,0],[5,1]]}]}]
      }
     ]
    }]}
 }]
graphs [
 EventGraph {nodes [
   OnActivated() {is=title K2Node_CustomEvent}
   > PlayAnimation(InAnimation=Get Intro())
   n1 = OnOrientationChanged!() -> (OutputDelegate, then, inScreenOrientation:EScreenOrientation Enum=Unknown) {is=title K2Node_CustomEvent}
   > UpdateLayoutOnOrientation!(execute, self, ScreenOrientation:EScreenOrientation Enum=@n1.inScreenOrientation) -> (then)
   Event Construct()
   > UpdateLayoutOnOrientation(ScreenOrientation=GetDeviceOrientation!() -> (ReturnValue:EScreenOrientation Enum=Unknown))
   > n2 = Cast To PlatformGameInstance!(execute, Object=GetGameInstance!(self:Widget Object Reference) -> (ReturnValue:Game Instance Object Reference)) -> (then, CastFailed, AsPlatform Game Instance:Platform Game Instance Object Reference) {is=title K2Node_DynamicCast}
   > n3 = Set As Platform Game Instance(As Platform Game Instance:Platform Game Instance Object Reference=@n2) -> (Output_Get:Platform Game Instance Object Reference)
   > Bind Event to Application Received Screen Orientation Changed Notification Delegate(self:Platform Game Instance Object Reference=@n3, Delegate=Create Event()) {is=title K2Node_AddDelegate}
   Event Destruct()
   n4 = Get As Platform Game Instance() -> (As Platform Game Instance:Platform Game Instance Object Reference)
   > Is Valid(InputObject=@n4)
   > Unbind Event from Application Received Screen Orientation Changed Notification Delegate(self:Platform Game Instance Object Reference=@n4, Delegate=@n1) {is=title K2Node_RemoveDelegate}
   "On Clicked (Button_Login)"!() -> (OutputDelegate, then, Button:Common Button Base Object Reference) {is=title K2Node_ComponentBoundEvent}
  ]}
 UpdateLayoutOnOrientation {nodes [
   n1 = UpdateLayoutOnOrientation!() -> (then, ScreenOrientation:EScreenOrientation Enum=Unknown) {is=title K2Node_FunctionEntry}
   n2 = GetViewportSize!() -> (ReturnValue_X:double, ReturnValue_Y:double)
   n3 = float / float(A=@n2.ReturnValue_Y, B=@n2)
   > n4 = Branch(Condition=AND Boolean(
     A="float > float"(A=@n3, B=1.34)
     B=Select!(
      Unknown:bool=false
      Portrait:bool=true
      PortraitUpsideDown:bool=true
      LandscapeLeft:bool=false
      LandscapeRight:bool=false
      FaceUp:bool=false
      FaceDown:bool=false
      PortraitSensor:bool=false
      LandscapeSensor:bool=false
      FullSensor:bool=false
      Index:EScreenOrientation Enum=@n1
     ) -> (ReturnValue:bool)
    ))
   n5 = SlotAsOverlaySlot(Widget=Get Button_Login() -> (Button_Login:Common Button Base Object Reference))
   > SetHorizontalAlignment(self=@n5, InHorizontalAlignment=HAlign_Fill)
   > SetPadding!(
    execute
    self:Overlay Slot Object Reference=@n5
    InPadding_Left:float=128
    InPadding_Top:float=0
    InPadding_Right:float=128
    InPadding_Bottom:float=256
   ) -> (then)
   SetHorizontalAlignment(execute=@n4.else, self=@n5, InHorizontalAlignment=HAlign_Center)
   > SetPadding(self=@n5, InPadding_Left=0, InPadding_Top=0, InPadding_Right=0, InPadding_Bottom=128)
  ]}
]
variables [As Platform Game Instance {type=PlatformGameInstance, default=null}]
widgets {
 info {inheritedWidgetCount=2, parentClass=/Script/FortniteUI.FortStartLoginWidget, rootWidgetClass=&SafeZone, widgetCount=5}
 widgets [
  SafeZone {widgetClassPath=&SafeZone, bIsVariable=true, widget=SafeZone, properties {padLeft=false, padRight=false, padTop=false}}
  Overlay_0 {widgetClassPath=&Overlay, parent=SafeZone, widget=Overlay_0, slot {class=SafeZoneSlot}}
  Button_Login {
   widgetClassPath=/FortUILibrary/UIKit/Buttons/WBP_UIKit_Button_Loud
   bInherited=true
   bIsVariable=true
   parent=Overlay_0
   widget=Button_Login
   properties {buttonSize=Large, minHeight=80, mobileMinHeight=64, mobileMinWidth=656, text=Sign in to Epic Games}
   slot {class=OverlaySlot, properties {horizontalAlignment=HAlign_Center, verticalAlignment=VAlign_Bottom, padding=<ltrb 0 0 0 128>}}
  }
  BackgroundBlur_PendingSpinner {
   widgetClassPath=&BackgroundBlur
   bInherited=true
   bIsVariable=true
   parent=Overlay_0
   widget=BackgroundBlur_PendingSpinner
   properties {
    blurStrength=3.5
    horizontalAlignment=HAlign_Center
    verticalAlignment=VAlign_Center
    visibility=Collapsed
    lowQualityFallbackBrush {
     bIsDynamicallyLoaded=false
     drawAs=Image
     imageType=NoImage
     mirroring=NoMirror
     resourceName=None
     resourceObject=null
     tiling=NoTile
     imageSize=<1024 768>
     margin=<ltrb 0 0 0 0>
     outlineSettings {
      bUseBrushTransparency=false
      roundingType=HalfHeightRadius
      width=0
      color {colorUseRule=UseColor_Specified, specifiedColor=<rgba 0 0 0 0>}
      cornerRadii=<0 0 0 0>
     }
     tintColor {colorUseRule=UseColor_Specified, specifiedColor=<rgba 0.0052 0.0052 0.0056 0.5>}
     uVRegion {bIsValid=false, max=<0 0>, min=<0 0>}
    }
    padding=<ltrb 0 0 0 80>
   }
   slot {class=OverlaySlot, properties {horizontalAlignment=HAlign_Fill, verticalAlignment=VAlign_Fill, padding=<ltrb -500 -500 -500 -500>}}
  }
  Spinner {
   widgetClassPath=&Image
   parent=BackgroundBlur_PendingSpinner
   widget=Spinner
   properties {
    brush {
     bIsDynamicallyLoaded=false
     drawAs=Image
     imageType=NoImage
     mirroring=NoMirror
     resourceName=None
     resourceObject=/FortUILibrary/UIKit/Throbber/MI_UIKit_Monochrome_Throbber
     tiling=NoTile
     imageSize=<32 32>
     margin=<ltrb 0 0 0 0>
     outlineSettings {
      bUseBrushTransparency=false
      roundingType=HalfHeightRadius
      width=0
      color {colorUseRule=UseColor_Specified, specifiedColor=<rgba 0 0 0 0>}
      cornerRadii=<0 0 0 0>
     }
     tintColor {colorUseRule=UseColor_Specified, specifiedColor=<rgba 1 1 1 1>}
     uVRegion {bIsValid=false, max=<0 0>, min=<0 0>}
    }
    renderTransform {angle=0, scale=<2.5 2.5>, shear=<0 0>, translation=<0 0>}
   }
   slot {class=BackgroundBlurSlot, properties {horizontalAlignment=HAlign_Center, verticalAlignment=VAlign_Center, padding=<ltrb 0 0 0 80>}}
  }
 ]
}
