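# UI-domain translucent material for a temporary Google login icon, built on the T_TempIcon_Login_Google texture parameter.
# The texture's RGB is desaturated, inverted and raised to a power set by the IconFilter parameter, forming a greyscale mask.
# An IconColor vector parameter, white by default, is also defined, with no connection shown from it or the mask to material outputs.
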Material /FortniteCoreUI/UI/Login/Art/M_TempIcon_Login_Google
graph {
 properties {
  bUsedWithStaticMesh=false
  blendMode=BLEND_Translucent
  materialDomain=MD_UI
  pixelDepthOffsetMode=PDOM_Legacy
  usedShadingModels=Default Lit
 }
 nodes [
  n1 = TextureSampleParameter2D() {parameterName=Texture, texture=&/T_TempIcon_Login_Google}
  Power(Base=OneMinus(None=Desaturation(None=@n1.RGB)), Exp=ScalarParameter() {defaultValue=0.2, parameterName=IconFilter}) {constExponent=4}
  VectorParameter() {parameterName=IconColor, defaultValue=<rgba 1 1 1 1>}
 ]
}
