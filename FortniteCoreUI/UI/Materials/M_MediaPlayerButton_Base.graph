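# Translucent UI base material that builds a media player button icon from a layered texture, scaled by the PlayButton scalar parameter.
# Rotated, offset samples of the G and B channels combine with a half-plane mask and sphere masks that cut and add circular shapes.
# Constant color nodes tint the summed shape cyan.
# The graph shows no wiring to material outputs, and several branches, including a DrawLine overlay, feed nothing further.
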Material /FortniteCoreUI/UI/Materials/M_MediaPlayerButton_Base
graph {
 properties {
  bUsedWithStaticMesh=false
  blendMode=BLEND_Translucent
  materialDomain=MD_UI
  pixelDepthOffsetMode=PDOM_Legacy
  usedShadingModels=Default Lit
 }
 nodes [
  n1 = TextureCoordinate()
  n2 = SphereMask(A=Subtract(A=Multiply(A=@n1) {constB=2}), B=Constant2Vector() {g=0.09, r=0.51}, Radius=Constant() {r=0.18}, Hardness=Constant() {r=0.9})
  n3 = TextureCoordinate()
  n4 = Constant() {r=-0.25}
  n5 = Multiply(
   A=OneMinus(None=Multiply(A=@n2))
   B=TextureSample(UVs=MaterialFunctionCall(UVs=@n3, "Rotation Angle (0-1)"=@n4) {materialFunction=&CustomRotator}.Rotated Values) {samplerSource=SSM_Clamp_WorldGroupSettings, texture=/FortniteCoreUI/UI/Icons/T_UI_MediaPlayerButton_Layered_A}.G
  )
  n6 = Ceil(None=Subtract(A=ComponentMask(None=@n3) {r=true}) {constB=0.48})
  n7 = MaterialFunctionCall(UVs=@n3, "Rotation Angle (0-1)"=Constant() {r=0.25}) {materialFunction=&CustomRotator}
  n8 = Constant() {r=0.25}
  n9 = Constant2Vector() {g=0.785, r=0.85}
  n10 = MaterialFunctionCall(Float2=MaterialFunctionCall(
    UVs=MaterialFunctionCall(UVs=@n3, "Rotation Angle (0-1)"=@n8) {materialFunction=&CustomRotator}.Rotated Values
    Texture Scale=@n9
   ) {materialFunction=&ScaleUVsByCenter}.UVs) {materialFunction=&BreakOutFloat2Components}
  n11 = Constant() {r=0.059}
  n12 = Add(A=Constant() {r=-0.06}, B=Divide(A=Multiply(B=@n11) {constA=0.81})) {constB=0}
  n13 = TextureSample(UVs=MaterialFunctionCall(X=@n10.R, Y=Subtract(A=@n10.G, B=@n12) {constB=0.1}) {materialFunction=&MakeFloat2}.Result) {samplerSource=SSM_Clamp_WorldGroupSettings, texture=&T_UI_MediaPlayerButton_Layered_A}
  n14 = Multiply(
   A=Multiply(
    A=OneMinus(None=@n6)
    B=TextureSample(UVs=@n7.Rotated Values) {samplerSource=SSM_Clamp_WorldGroupSettings, texture=&T_UI_MediaPlayerButton_Layered_A}.B
   )
   B=OneMinus(None=Power(Base=@n13.G))
  )
  n15 = TextureCoordinate()
  n16 = MaterialFunctionCall(Float2=Constant2Vector() {g=-0.365, r=-0.51}) {materialFunction=&BreakOutFloat2Components}
  n17 = SphereMask(
   A=Subtract(A=Multiply(A=@n15) {constB=2})
   B=MaterialFunctionCall(X=Add(A=@n16.R, B=@n11), Y=@n16.G) {materialFunction=&MakeFloat2}.Result
   Radius=Constant() {r=0.17}
   Hardness=Constant() {r=0.9}
  )
  n18 = Multiply(A=Add(A=Add(A=@n5, B=@n14), B=Multiply(A=@n17)), B=ScalarParameter() {parameterName=PlayButton})
  n19 = Constant3Vector() {constant=<rgba 1 1 1 0>}
  n20 = Multiply(B=Constant3Vector() {constant=<rgba 1 0 0 0>})
  Add(A=Multiply(A=Constant3Vector() {constant=<rgba 0 0.9576 1 0>}, B=Multiply(A=@n18, B=@n19)), B=@n20)
  Add(
   A=@n18
   B=MaterialFunctionCall!(Coordinates, P0=Constant2Vector() {r=0.5}, P1=Constant2Vector() {g=1, r=0.5}, Line Width, Edge Falloff, Square Corners?) {materialFunction=/Engine/Functions/Engine_MaterialFunctions01/Coordinates/DrawLine-2D}.Result
  )
  MaterialFunctionCall(Float2=@n9) {materialFunction=&BreakOutFloat2Components}
  MaterialFunctionCall(UVs=@n7.Rotated Values, Texture Scale=Constant2Vector() {g=1.165, r=1}) {materialFunction=&ScaleUVsByCenter}
  Multiply(A=@n20, B=@n18)
  Multiply(A=@n20)
 ]
}
